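; B_WaterCausticVolume is a box volume that applies water caustic material effects to the BuildingSMActors it overlaps.
; Its construction script diffs current overlaps against history, removing owned B_WaterCausticVolume_Components from departed actors and adding or claiming them on new ones with at most two static meshes.
; A lower Priority value takes over components from other volumes.
; It then pushes box bounds, emissive, falloff and hue settings to its components for updating.
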(event EventBeginPlay)

(event Collision|EventActorBeginOverlap
  (params (OtherActor :type Actor%20Object%20Reference)))

(event EventTick
  (params (DeltaSeconds :type Float%20%28single-precision%29)))

(event Custom|Update%20Water%20Caustic
  (params (BuildingSMActor :type Actor%20Object%20Reference))
  (Utilities|IsValid (Actor|GetComponentbyClass BuildingSMActor "/Game/ArtTools/Tech_Art/WaterCausticVolume/B_WaterCausticVolume_Component.B_WaterCausticVolume_Component_C")))

(fn AddComponents
  (for _array_element (Variables|Default|GetAddQueue)
    (bind _asbuilding_static_mesh (Utilities|Casting|CastToBuildingSMActor _array_element))
    (Development|PrintString 0 (Utilities|String|ToString(Integer) (Utilities|Array|Length (Actor|GetComponentsByClass _asbuilding_static_mesh "/Script/Engine.StaticMeshComponent"))))
    (if (<= (Utilities|Array|Length (Actor|GetComponentsByClass _asbuilding_static_mesh "/Script/Engine.StaticMeshComponent")) 2)
      (Utilities|IsValid (Actor|GetComponentbyClass _asbuilding_static_mesh "/Game/ArtTools/Tech_Art/WaterCausticVolume/B_WaterCausticVolume_Component.B_WaterCausticVolume_Component_C")
        (:Is%20Valid
          (if (or (not (Utilities|IsValid (Variables|/Game/ArtTools/Tech_Art/WaterCausticVolume/B_WaterCausticVolume_Component.B_WaterCausticVolume_Component_C|GetVolumeActor (Actor|GetComponentbyClass _asbuilding_static_mesh "/Game/ArtTools/Tech_Art/WaterCausticVolume/B_WaterCausticVolume_Component.B_WaterCausticVolume_Component_C")) :!class /Script/Engine.KismetSystemLibrary)) (< (Variables|Default|GetPriority) (Variables|Default|GetPriority (Variables|/Game/ArtTools/Tech_Art/WaterCausticVolume/B_WaterCausticVolume_Component.B_WaterCausticVolume_Component_C|GetVolumeActor (Actor|GetComponentbyClass _asbuilding_static_mesh "/Game/ArtTools/Tech_Art/WaterCausticVolume/B_WaterCausticVolume_Component.B_WaterCausticVolume_Component_C")))))
            (Variables|/Game/ArtTools/Tech_Art/WaterCausticVolume/B_WaterCausticVolume_Component.B_WaterCausticVolume_Component_C|SetVolumeActor self (Actor|GetComponentbyClass _asbuilding_static_mesh "/Game/ArtTools/Tech_Art/WaterCausticVolume/B_WaterCausticVolume_Component.B_WaterCausticVolume_Component_C"))
            (Utilities|Array|AddUnique (Variables|Default|GetVolumeComponents) (Actor|GetComponentbyClass _asbuilding_static_mesh "/Game/ArtTools/Tech_Art/WaterCausticVolume/B_WaterCausticVolume_Component.B_WaterCausticVolume_Component_C"))))
        (:Is%20Not%20Valid
          (bind _returnvalue (AddComponent|NODE_AddB_WaterCausticVolume_Component-0 :self _asbuilding_static_mesh))
          (Class|BWaterCausticVolumeComponent|SaveMaterial _returnvalue)
          (Variables|/Game/ArtTools/Tech_Art/WaterCausticVolume/B_WaterCausticVolume_Component.B_WaterCausticVolume_Component_C|SetVolumeActor self _returnvalue)
          (Utilities|Array|AddUnique (Variables|Default|GetVolumeComponents) _returnvalue)
          (Variables|/Script/FortniteGame.BuildingSMActor|SetbAllowCustomMaterial true _asbuilding_static_mesh)))))
  (return))

(fn CompareNewToHistory
  (Utilities|Array|Clear (Variables|Default|GetDeleteQueue))
  (Variables|Default|SetAddQueue (Variables|Default|GetCausticENVActorsNew))
  (for _array_element (Variables|Default|GetCausticENVActorsHistory)
    (if (Utilities|Array|ContainsItem (Variables|Default|GetCausticENVActorsNew) _array_element)
      (Utilities|Array|RemoveItem (Variables|Default|GetAddQueue) _array_element)
      (else
        (Utilities|Array|Add (Variables|Default|GetDeleteQueue) _array_element))))
  (Variables|Default|SetCausticENVActorsHistory (Variables|Default|GetCausticENVActorsNew))
  (return))

(fn DeleteComponents
  (for _array_element (Variables|Default|GetDeleteQueue)
    (bind _asbuilding_static_mesh (Utilities|Casting|CastToBuildingSMActor _array_element))
    (Utilities|IsValid (Actor|GetComponentbyClass _asbuilding_static_mesh "/Game/ArtTools/Tech_Art/WaterCausticVolume/B_WaterCausticVolume_Component.B_WaterCausticVolume_Component_C")
      (:Is%20Valid
        (if (== (Variables|/Game/ArtTools/Tech_Art/WaterCausticVolume/B_WaterCausticVolume_Component.B_WaterCausticVolume_Component_C|GetVolumeActor (Actor|GetComponentbyClass _asbuilding_static_mesh "/Game/ArtTools/Tech_Art/WaterCausticVolume/B_WaterCausticVolume_Component.B_WaterCausticVolume_Component_C")) self)
          (Utilities|Array|RemoveItem (Variables|Default|GetVolumeComponents) (Actor|GetComponentbyClass _asbuilding_static_mesh "/Game/ArtTools/Tech_Art/WaterCausticVolume/B_WaterCausticVolume_Component.B_WaterCausticVolume_Component_C"))
          (Class|BWaterCausticVolumeComponent|DestroySelf (Actor|GetComponentbyClass _asbuilding_static_mesh "/Game/ArtTools/Tech_Art/WaterCausticVolume/B_WaterCausticVolume_Component.B_WaterCausticVolume_Component_C"))))))
  (return))

(fn GetOverlappingBuildingActors
  (bind _outactors (Collision|ComponentOverlapActors (Variables|Default|GetBox) (Transformation|GetWorldTransform (Variables|Default|GetBox)) (Variables|Default|GetObject%20Types) "/Script/FortniteGame.BuildingSMActor"))
  (bind _output_get (Variables|Default|SetCausticENVActorsNew _outactors))
  (return _output_get))

(fn RemoveWaterCaustics
  (for _array_element (Variables|Default|GetVolumeComponents)
    (Variables|/Game/ArtTools/Tech_Art/WaterCausticVolume/B_WaterCausticVolume_Component.B_WaterCausticVolume_Component_C|SetVolumeActor self _array_element)
    (Class|BWaterCausticVolumeComponent|ApplyOrignalVisual _array_element))
  (return))

(fn SaveOriginalMaterials
  (params (BuildingSM :type Primitive%20Component%20Object%20Reference) (SMIndex :type Integer))
  (bind _returnvalue (Rendering|Material|GetMaterial BuildingSM 0))
  (bind _asmaterial_instance_dynamic (Utilities|Casting|CastToMaterialInstanceDynamic _returnvalue)
    (:then
      (Variables|Default|SetMaterial_Original _asmaterial_instance_dynamic))
    (:CastFailed
      (bind _asmaterial_instance_constant (Utilities|Casting|CastToMaterialInstanceConstant _returnvalue))
      (Variables|Default|SetMaterial_Original _asmaterial_instance_constant))))

(fn SetMaterialParameters
  (bind _material_new (Variables|Default|GetMaterial_New))
  (bind _material_new_2 (Variables|Default|GetMaterial_New))
  (for _array_element (Variables|Default|GetCausticENVActorsNew)
    (bind _asbuilding_static_mesh (Utilities|Casting|CastToBuildingSMActor _array_element))
    (bind _asmaterial_instance_dynamic (Utilities|Casting|CastToMaterialInstanceDynamic (Rendering|Material|GetMaterial (Variables|/Script/FortniteGame.BuildingSMActor|GetStaticMeshComponent _asbuilding_static_mesh) 0)))
    (Variables|Default|SetMaterial_New _asmaterial_instance_dynamic)
    ;[0.02613 0.03473 0.07812] Change Caustic Material Parameters
    (Rendering|Material|SetScalarParameterValue _material_new "Enable Water Caustic" 1.0 :!class /Script/Engine.MaterialInstanceDynamic)
    (Rendering|Material|SetVectorParameterValue _material_new "CausticVolumeOrigin" (Math|Conversions|ToLinearColor(Vector) (Variables|Default|GetCausticVolumeOrigin)) :!class /Script/Engine.MaterialInstanceDynamic)
    (Rendering|Material|SetVectorParameterValue _material_new "CausticVolumeBounds" (Math|Conversions|ToLinearColor(Vector) (Variables|Default|GetCausticVolumeBounds)) :!class /Script/Engine.MaterialInstanceDynamic)
    ;[0.142 0.5845 1] Apply texture data back
    (bind _outtexturedataarray (Building|GetBuildingTextureData _asbuilding_static_mesh :!class /Script/FortniteGame.BuildingSMActor))
    (for (_array_element_2 _array_index) _outtexturedataarray
      (Rendering|Material|SetTextureParameterValue _material_new_2 (Utilities|Array|Get(aref) (Variables|Default|GetDiffuseParameterNames) _array_index) (Variables|/Script/FortniteGame.BuildingTextureData|GetDiffuse _array_element_2))
      (Rendering|Material|SetTextureParameterValue _material_new_2 (Utilities|Array|Get(aref) (Variables|Default|GetSpecularParameterNames) _array_index) (Variables|/Script/FortniteGame.BuildingTextureData|GetSpecular _array_element_2))
      (Rendering|Material|SetTextureParameterValue _material_new_2 (Utilities|Array|Get(aref) (Variables|Default|GetNormalParameterNames) _array_index) (Variables|/Script/FortniteGame.BuildingTextureData|GetNormal _array_element_2)))))

(fn UpdateComponentVariables
  (for _array_element (Variables|Default|GetVolumeComponents)
    (Utilities|IsValid _array_element
      (:Is%20Valid
        (Variables|/Game/ArtTools/Tech_Art/WaterCausticVolume/B_WaterCausticVolume_Component.B_WaterCausticVolume_Component_C|SetVolumeActor self _array_element)
        (Variables|/Game/ArtTools/Tech_Art/WaterCausticVolume/B_WaterCausticVolume_Component.B_WaterCausticVolume_Component_C|SetEnable%20Water%20Caustic (Variables|Default|GetEnable%20Water%20Caustic) _array_element)
        (Variables|/Game/ArtTools/Tech_Art/WaterCausticVolume/B_WaterCausticVolume_Component.B_WaterCausticVolume_Component_C|SetCausticVolumeOrigin (Variables|Default|GetCausticVolumeOrigin) _array_element)
        (Variables|/Game/ArtTools/Tech_Art/WaterCausticVolume/B_WaterCausticVolume_Component.B_WaterCausticVolume_Component_C|SetCausticVolumeBounds (Variables|Default|GetCausticVolumeBounds) _array_element)
        (Variables|/Game/ArtTools/Tech_Art/WaterCausticVolume/B_WaterCausticVolume_Component.B_WaterCausticVolume_Component_C|SetWaterCausticEmissive_Min (Variables|Default|GetWaterCausticEmissive_Min) _array_element)
        (Variables|/Game/ArtTools/Tech_Art/WaterCausticVolume/B_WaterCausticVolume_Component.B_WaterCausticVolume_Component_C|SetWaterCausticEmissive_Max (Variables|Default|GetWaterCausticEmissive_Max) _array_element)
        (Variables|/Game/ArtTools/Tech_Art/WaterCausticVolume/B_WaterCausticVolume_Component.B_WaterCausticVolume_Component_C|SetCausticsMaskFallOff_X (Variables|Default|GetCausticsMaskFallOff_X) _array_element)
        (Variables|/Game/ArtTools/Tech_Art/WaterCausticVolume/B_WaterCausticVolume_Component.B_WaterCausticVolume_Component_C|SetCausticsMaskFallOff_Y (Variables|Default|GetCausticsMaskFallOff_Y) _array_element)
        (Variables|/Game/ArtTools/Tech_Art/WaterCausticVolume/B_WaterCausticVolume_Component.B_WaterCausticVolume_Component_C|SetCausticsMaskFallOff_Z (Variables|Default|GetCausticsMaskFallOff_Z) _array_element)
        (Variables|/Game/ArtTools/Tech_Art/WaterCausticVolume/B_WaterCausticVolume_Component.B_WaterCausticVolume_Component_C|SetCausticColor_Hueshift (Variables|Default|GetCausticColor_Hueshift) _array_element)
        (Class|BWaterCausticVolumeComponent|UpdateWaterCaustic _array_element))))
  (return))

(fn ConstructionScript
  (bind _box (Variables|Default|GetBox))
  (bind (_origin _boxextent) (Collision|GetComponentBounds _box))
  (Rendering|SetVisibility (Variables|Default|GetCube) (Variables|Default|GetShow%20Preview%20Box) false)
  (CallFunction|GetOverlappingBuildingActors)
  (Variables|Default|SetCausticVolumeOrigin _origin)
  (Variables|Default|SetCausticVolumeBounds _boxextent)
  (CallFunction|CompareNewtoHistory)
  (CallFunction|DeleteComponents)
  (CallFunction|AddComponents)
  (CallFunction|UpdateComponentVariables))
