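; Building actor component that swaps a jungle static mesh's material for a water-caustic variant and can restore it.
; SaveMaterial caches the owner's material, which DestroySelf reapplies before removing the component.
; UpdateWaterCaustic name-matches the mesh material against parallel parent arrays, creates a MID, then copies building textures and caustic volume settings.
; ApplyOrignalVisual rebuilds from OriginalMaterialParent (the matched caustic parent), but writes caustic parameters to WaterCausticMID.
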(event Custom|UpdateWaterCaustic
  (CallFunction|CreateNewMID)
  (CallFunction|ApplyMaterialParameters))

(event Custom|DestroySelf
  (Rendering|Material|SetMaterial (Variables|/Script/FortniteGame.BuildingSMActor|GetStaticMeshComponent (Variables|Default|GetBuildingSMParent)) 0 (Variables|Default|GetOriginalMaterial))
  (Components|DestroyComponent))

(event Custom|ApplyOrignalVisual
  (CallFunction|CreateMIDfromOriginalParent)
  (CallFunction|ApplyMaterialParametersOriginalParentMID))

(fn ApplyMaterialParameters
  (bind _watercausticmid (Variables|Default|GetWaterCausticMID))
  (bind _watercausticmid_2 (Variables|Default|GetWaterCausticMID))
  (bind _outtexturedataarray (Building|GetBuildingTextureData (Variables|Default|GetBuildingSMParent) :!class /Script/FortniteGame.BuildingSMActor))
  (for (_array_element _array_index) _outtexturedataarray
    (Rendering|Material|SetTextureParameterValue _watercausticmid (Utilities|Array|Get(aref) (Variables|Default|GetDiffuseParameterNames) _array_index) (Variables|/Script/FortniteGame.BuildingTextureData|GetDiffuse _array_element))
    (Rendering|Material|SetTextureParameterValue _watercausticmid (Utilities|Array|Get(aref) (Variables|Default|GetSpecularParameterNames) _array_index) (Variables|/Script/FortniteGame.BuildingTextureData|GetSpecular _array_element))
    (Rendering|Material|SetTextureParameterValue _watercausticmid (Utilities|Array|Get(aref) (Variables|Default|GetNormalParameterNames) _array_index) (Variables|/Script/FortniteGame.BuildingTextureData|GetNormal _array_element)))
  (Rendering|Material|SetScalarParameterValue _watercausticmid_2 "Enable Water Caustic" (Math|Conversions|ToFloat(Boolean) (Variables|Default|GetEnable%20Water%20Caustic)) :!class /Script/Engine.MaterialInstanceDynamic)
  (Rendering|Material|SetVectorParameterValue _watercausticmid_2 "CausticVolumeOrigin" (Math|Conversions|ToLinearColor(Vector) (Variables|Default|GetCausticVolumeOrigin)) :!class /Script/Engine.MaterialInstanceDynamic)
  (Rendering|Material|SetVectorParameterValue _watercausticmid_2 "CausticVolumeBounds" (Math|Conversions|ToLinearColor(Vector) (Variables|Default|GetCausticVolumeBounds)) :!class /Script/Engine.MaterialInstanceDynamic)
  (Rendering|Material|SetScalarParameterValue _watercausticmid_2 "CausticsBrightness_Min" (Variables|Default|GetWaterCausticEmissive_Min) :!class /Script/Engine.MaterialInstanceDynamic)
  (Rendering|Material|SetScalarParameterValue _watercausticmid_2 "CausticsBrightness_Max" (Variables|Default|GetWaterCausticEmissive_Max) :!class /Script/Engine.MaterialInstanceDynamic)
  (Rendering|Material|SetScalarParameterValue _watercausticmid_2 "CausticsMaskFallOff_X" (Variables|Default|GetCausticsMaskFallOff_X) :!class /Script/Engine.MaterialInstanceDynamic)
  (Rendering|Material|SetScalarParameterValue _watercausticmid_2 "CausticsMaskFallOff_Y" (Variables|Default|GetCausticsMaskFallOff_Y) :!class /Script/Engine.MaterialInstanceDynamic)
  (Rendering|Material|SetScalarParameterValue _watercausticmid_2 "CausticsMaskFallOff_Z" (Variables|Default|GetCausticsMaskFallOff_Z) :!class /Script/Engine.MaterialInstanceDynamic)
  (Rendering|Material|SetScalarParameterValue _watercausticmid_2 "CausticColor_Hueshift" (Variables|Default|GetCausticColor_Hueshift) :!class /Script/Engine.MaterialInstanceDynamic))

(fn ApplyMaterialParameters_OriginalParentMID
  (bind _originalparentmid (Variables|Default|GetOriginalParentMID))
  (bind _watercausticmid (Variables|Default|GetWaterCausticMID))
  (bind _outtexturedataarray (Building|GetBuildingTextureData (Variables|Default|GetBuildingSMParent) :!class /Script/FortniteGame.BuildingSMActor))
  (for (_array_element _array_index) _outtexturedataarray
    (Rendering|Material|SetTextureParameterValue _originalparentmid (Utilities|Array|Get(aref) (Variables|Default|GetDiffuseParameterNames) _array_index) (Variables|/Script/FortniteGame.BuildingTextureData|GetDiffuse _array_element))
    (Rendering|Material|SetTextureParameterValue _originalparentmid (Utilities|Array|Get(aref) (Variables|Default|GetSpecularParameterNames) _array_index) (Variables|/Script/FortniteGame.BuildingTextureData|GetSpecular _array_element))
    (Rendering|Material|SetTextureParameterValue _originalparentmid (Utilities|Array|Get(aref) (Variables|Default|GetNormalParameterNames) _array_index) (Variables|/Script/FortniteGame.BuildingTextureData|GetNormal _array_element)))
  (Rendering|Material|SetScalarParameterValue _watercausticmid "Enable Water Caustic" (Math|Conversions|ToFloat(Boolean) (Variables|Default|GetEnable%20Water%20Caustic)) :!class /Script/Engine.MaterialInstanceDynamic)
  (Rendering|Material|SetVectorParameterValue _watercausticmid "CausticVolumeOrigin" (Math|Conversions|ToLinearColor(Vector) (Variables|Default|GetCausticVolumeOrigin)) :!class /Script/Engine.MaterialInstanceDynamic)
  (Rendering|Material|SetVectorParameterValue _watercausticmid "CausticVolumeBounds" (Math|Conversions|ToLinearColor(Vector) (Variables|Default|GetCausticVolumeBounds)) :!class /Script/Engine.MaterialInstanceDynamic)
  (Rendering|Material|SetScalarParameterValue _watercausticmid "CausticsBrightness_Min" (Variables|Default|GetWaterCausticEmissive_Min) :!class /Script/Engine.MaterialInstanceDynamic)
  (Rendering|Material|SetScalarParameterValue _watercausticmid "CausticsBrightness_Max" (Variables|Default|GetWaterCausticEmissive_Max) :!class /Script/Engine.MaterialInstanceDynamic))

(fn CreateMIDfromOriginalParent
  (bind _returnvalue (Rendering|Material|CreateDynamicMaterialInstance (Variables|Default|GetOriginalMaterialParent) "None" "None" :!class /Script/Engine.KismetMaterialLibrary))
  (bind _output_get (Variables|Default|SetOriginalParentMID _returnvalue))
  (Rendering|Material|SetMaterial (Variables|/Script/FortniteGame.BuildingSMActor|GetStaticMeshComponent (Variables|Default|GetBuildingSMParent)) 0 _output_get))

(fn CreateNewMID
  (bind _buildingsmparent (Variables|Default|GetBuildingSMParent))
  (bind _staticmeshcomponent (Variables|/Script/FortniteGame.BuildingSMActor|GetStaticMeshComponent _buildingsmparent))
  (bind (_array_element _array_index) (Utilities|Array|ForEachLoopwithBreak (Variables|Default|GetJungleParentMIs))
    (:LoopBody
      (if (Utilities|String|EqualExactly(String) (Utilities|GetDisplayName (StaticMesh|GetMaterial (Variables|/Script/Engine.StaticMeshComponent|GetStaticMesh _staticmeshcomponent) 0)) (Utilities|GetDisplayName _array_element))
        (bind _output_get (Variables|Default|SetJungleParentMatchingIndex _array_index))
        (bind _output_get_2 (Variables|Default|SetOriginalMaterialParent (Utilities|Array|Get(acopy) (Variables|Default|GetJungleParentMIs_Caustic) _output_get)))
        (bind _returnvalue (Rendering|Material|CreateDynamicMaterialInstance _output_get_2 "None" "None" :!class /Script/Engine.KismetMaterialLibrary))
        (bind _output_get_3 (Variables|Default|SetWaterCausticMID _returnvalue))
        (Rendering|Material|SetMaterial _staticmeshcomponent 0 _output_get_3)))))

(fn SaveMaterial
  (bind _asbuilding_static_mesh (Utilities|Casting|CastToBuildingSMActor (Components|GetOwner)))
  (Variables|Default|SetBuildingSMParent _asbuilding_static_mesh)
  (for _array_element (Actor|GetComponentsByClass (Variables|Default|GetBuildingSMParent) "/Script/Engine.StaticMeshComponent")
    (bind _asmaterial_instance_dynamic (Utilities|Casting|CastToMaterialInstanceDynamic (Rendering|Material|GetMaterial _array_element 0))
      (:then
        (Variables|Default|SetOriginalMaterial _asmaterial_instance_dynamic))
      (:CastFailed
        (bind _asmaterial_instance_constant (Utilities|Casting|CastToMaterialInstanceConstant (Rendering|Material|GetMaterial _array_element 0)))
        (bind _output_get (Variables|Default|SetOriginalMaterial _asmaterial_instance_constant))))))
